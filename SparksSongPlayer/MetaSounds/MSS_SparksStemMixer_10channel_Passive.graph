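# Quad MetaSound source that mixes Sparks song stems read from pre-effect stereo buses, using its Music Stems Player only for transport and MIDI clock.
# Five interpolated-gain stems feed the front pair and the Fake rear bus the rear, scaled by fade envelope, volume trim and lobby music slider.
# The combined mix drives peak, four-band and FFT analysis outputs.
# OnFinished fires from envelope triggers or MIDI length reached when not looping.
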MetaSoundSource /SparksSongPlayer/MetaSounds/MSS_SparksStemMixer_10channel_Passive
inputs [
 Bass Gain {type=Float}
 Bass Stereo Bus {type=AudioBusAsset}
 bLooping {type=Bool}
 Drums Gain {type=Float}
 Drums Stereo Bus {type=AudioBusAsset}
 Fade In Duration {type=Time}
 Fade Out Duration {type=Time}
 Fake Gain {type=Float}
 Fake Rear Stereo Bus {type=AudioBusAsset}
 Fake Stereo Bus {type=AudioBusAsset}
 Guitar Gain {type=Float}
 Guitars Stereo Bus {type=AudioBusAsset}
 Seek Target {type=MusicSeekTarget}
 Speed {type=Float}
 Stems Asset {type=StemsAsset, access=Constructor}
 Track Fade Duration {type=Time}
 Trigger Seek {type=Trigger}
 TriggerFadeOut {type=Trigger}
 UE.Source.OnPlay {type=Trigger}
 Vocals Gain {type=Float}
 Vocals Stereo Bus {type=AudioBusAsset}
 Volume Adjust Db {type=Float}
]
interfaces [UE.OutputFormat.Quad:1.0, UE.Source.OneShot:1.0, UE.Source:1.1]
outputs [
 Amplitude {type=Float}
 Band1 {type=Float}
 Band2 {type=Float}
 Band3 {type=Float}
 Band4 {type=Float}
 FFTResults {type=HarmonixFFTAnalyzerResults}
 MIDI Clock {type=MIDIClock}
 "UE.OutputFormat.Quad.Audio:0" {type=Audio}
 "UE.OutputFormat.Quad.Audio:1" {type=Audio}
 "UE.OutputFormat.Quad.Audio:4" {type=Audio}
 "UE.OutputFormat.Quad.Audio:5" {type=Audio}
 UE.Source.OneShot.OnFinished {type=Trigger}
]
pages [Default {
  input_defaults {
   Bass Gain=1
   Bass Stereo Bus=/SparksSongPlayer/MetaSounds/Buses/Pre-Effect/BassStereoBus
   Drums Gain=1
   Drums Stereo Bus=/SparksSongPlayer/MetaSounds/Buses/Pre-Effect/DrumsStereoBus
   Fade In Duration=0.15
   Fade Out Duration=0.5
   Fake Gain=1
   Fake Rear Stereo Bus=/SparksSongPlayer/MetaSounds/Buses/Pre-Effect/FakeRearStereoBus
   Fake Stereo Bus=/SparksSongPlayer/MetaSounds/Buses/Pre-Effect/FakeStereoBus
   Guitar Gain=1
   Guitars Stereo Bus=/SparksSongPlayer/MetaSounds/Buses/Pre-Effect/GuitarsStereoBus
   Seek Target=null
   Speed=1
   Stems Asset=null
   Track Fade Duration=0.001
   Trigger Seek=null
   TriggerFadeOut=null
   UE.Source.OnPlay=null
   Vocals Gain=1
   Vocals Stereo Bus=/SparksSongPlayer/MetaSounds/Buses/Pre-Effect/VocalsStereoBus
   Volume Adjust Db=-1
   bLooping=null
  }
  nodes [
   AudioBusReader2_0 = "UE.Audio Bus Reader (2)"(Audio Bus=^Fake Rear Stereo Bus)
   n1 = UE.Reroute(Value=^Track Fade Duration)
   n2 = UE.Reroute(Value=@n1.Value)
   n3 = UE.Reroute(Value=@n2.Value)
   n4 = UE.Reroute(Value=@n3.Value)
   n5 = UE.Reroute(Value=@n4.Value)
   n6 = UE.Reroute(Value=UE.Reroute(Value=UE.InterpTo:Audio(Target=^Fake Gain, Interp Time=@n5.Value).Value).Value)
   AudioMixerAudioMixerStereo2_1 = "AudioMixer.Audio Mixer (Stereo, 2)"(In 0 L=@AudioBusReader2_0.Out 0, In 0 R=@AudioBusReader2_0.Out 1, Gain 0=@n6.Value)
   n7 = /Game/Sounds/Modulation/Buses/Sliders/MSP_LobbyMusicMultiplier(FrontEndAttenuationDecibels=VariableAccessor.Float(Variable=InitVariable.Float().Variable).Value)
   DecibelstoLinearGain_0 = UE.Decibels to Linear Gain:Float(Decibels=^Volume Adjust Db)
   HarmonixNodesMusicStemsPlayer_0 = HarmonixNodes.Music Stems Player(
    Speed=^Speed
    Stems Asset=^Stems Asset
    Transport=HarmonixNodes.TriggerToTransport(Seek Target=^Seek Target, Trigger Seek=^Trigger Seek, Play=^UE.Source.OnPlay).Transport
   )
   ADSREnvelopeADSREnvelope_0 = ADSR Envelope.ADSR Envelope:Float(
    Attack Time=^Fade In Duration
    Release Time=^Fade Out Duration
    Trigger Release=^TriggerFadeOut
    Trigger Attack=HarmonixNodes.TransportToTrigger(Transport=@HarmonixNodesMusicStemsPlayer_0.Transport) {also {version [0,1]}}.Play
   ) {also {defaults {Attack Curve=2, Decay Time=0, Sustain Level=1}}}
   n8 = UE.Reroute(Value=@ADSREnvelopeADSREnvelope_0.Out Envelope)
   n9 = UE.Multiply:Float(PrimaryOperand=@DecibelstoLinearGain_0.Linear Gain, AdditionalOperands=UE.Reroute(Value=UE.Reroute(Value=@n8.Value).Value).Value)
   n10 = VariableAccessor.Float(Variable=VariableMutator.Float(
     Variable=InitVariable.Float().Variable
     Value=UE.Multiply:Float(PrimaryOperand=@n7.NormalizedLobbyMusicSlider, AdditionalOperands=@n9.Out).Out
    ).Variable)
   n11 = VariableAccessor.Float(Variable=@n10.Variable)
   Reroute_0 = UE.Reroute(Value=UE.Multiply:Audio by Float(PrimaryOperand=@AudioMixerAudioMixerStereo2_1.Out L, AdditionalOperands=@n11.Value).Out)
   Reroute_27 = UE.Reroute(Value=UE.Multiply:Audio by Float(PrimaryOperand=@AudioMixerAudioMixerStereo2_1.Out R, AdditionalOperands=@n11.Value).Out)
   AudioBusReader2_1 = "UE.Audio Bus Reader (2)"(Audio Bus=^Vocals Stereo Bus)
   AudioBusReader2_3 = "UE.Audio Bus Reader (2)"(Audio Bus=^Fake Stereo Bus)
   n12 = UE.InterpTo:Audio(Target=^Vocals Gain, Interp Time=@n4.Value)
   AudioBusReader2_2 = "UE.Audio Bus Reader (2)"(Audio Bus=^Bass Stereo Bus)
   AudioBusReader2_5 = "UE.Audio Bus Reader (2)"(Audio Bus=^Guitars Stereo Bus)
   AudioBusReader2_4 = "UE.Audio Bus Reader (2)"(Audio Bus=^Drums Stereo Bus)
   n13 = UE.Reroute(Value=UE.InterpTo:Audio(Target=^Bass Gain, Interp Time=@n2.Value).Value)
   n14 = UE.Reroute(Value=UE.InterpTo:Audio(Target=^Drums Gain, Interp Time=@n1.Value).Value)
   n15 = UE.InterpTo:Audio(Target=^Guitar Gain, Interp Time=@n3.Value)
   AudioMixerAudioMixerStereo5_0 = "AudioMixer.Audio Mixer (Stereo, 5)"(
    In 3 L=@AudioBusReader2_1.Out 0
    In 3 R=@AudioBusReader2_1.Out 1
    In 4 L=@AudioBusReader2_3.Out 0
    In 4 R=@AudioBusReader2_3.Out 1
    Gain 3=UE.Reroute(Value=UE.Reroute(Value=@n12.Value).Value).Value
    Gain 4=@n6.Value
    In 1 L=UE.Reroute(Value=@AudioBusReader2_2.Out 0).Value
    In 2 L=UE.Reroute(Value=@AudioBusReader2_5.Out 0).Value
    In 0 R=UE.Reroute(Value=@AudioBusReader2_4.Out 1).Value
    Gain 1=UE.Reroute(Value=UE.Reroute(Value=@n13.Value).Value).Value
    In 1 R=UE.Reroute(Value=@AudioBusReader2_2.Out 1).Value
    In 0 L=UE.Reroute(Value=@AudioBusReader2_4.Out 0).Value
    In 2 R=UE.Reroute(Value=@AudioBusReader2_5.Out 1).Value
    Gain 0=UE.Reroute(Value=UE.Reroute(Value=@n14.Value).Value).Value
    Gain 2=UE.Reroute(Value=UE.Reroute(Value=@n15.Value).Value).Value
   )
   Reroute_3 = UE.Reroute(Value=UE.Multiply:Audio by Float(PrimaryOperand=@AudioMixerAudioMixerStereo5_0.Out R, AdditionalOperands=@n10.Value).Out)
   Reroute_32 = UE.Reroute(Value=UE.Multiply:Audio by Float(PrimaryOperand=@AudioMixerAudioMixerStereo5_0.Out L, AdditionalOperands=@n10.Value).Out)
   AudioMixerAudioMixerStereo2_0 = "AudioMixer.Audio Mixer (Stereo, 2)"(In 1 L=@Reroute_0.Value, In 1 R=@Reroute_27.Value, In 0 R=@Reroute_3.Value, In 0 L=@Reroute_32.Value)
   n16 = HarmonixNodes.MultibandAnalyzer(Audio Mono=UE.Add:Audio(PrimaryOperand=@AudioMixerAudioMixerStereo2_0.Out L, AdditionalOperands=@AudioMixerAudioMixerStereo2_0.Out R).Out) {also {defaults {Release Time=0.1}, version [0,1]}}
   n17 = UE.TriggerRepeat(Start=^UE.Source.OnPlay) {also {defaults {Period=0.0333}}}
   Divide_0 = UE.Divide:Float(PrimaryOperand=Array.Get:Float:Array(Array=@n16.Band Levels, Trigger=@n17.RepeatOut) {also {defaults {Index=3}}}.Element) {also {defaults {AdditionalOperands=2}}}
   Divide_1 = UE.Divide:Float(PrimaryOperand=Array.Get:Float:Array(Array=@n16.Band Levels, Trigger=@n17.RepeatOut) {also {defaults {Index=2}}}.Element) {also {defaults {AdditionalOperands=2}}}
   Divide_2 = UE.Divide:Float(PrimaryOperand=Array.Get:Float:Array(Array=@n16.Band Levels, Trigger=@n17.RepeatOut).Element) {also {defaults {AdditionalOperands=2}}}
   Divide_3 = UE.Divide:Float(PrimaryOperand=HarmonixNodes.Peak(Audio Mono=UE.Add:Audio(PrimaryOperand=@AudioMixerAudioMixerStereo2_0.Out L, AdditionalOperands=@AudioMixerAudioMixerStereo2_0.Out R).Out) {also {version [0,1]}}.Peak) {also {defaults {AdditionalOperands=2}}}
   Divide_4 = UE.Divide:Float(PrimaryOperand=Array.Get:Float:Array(Array=@n16.Band Levels, Trigger=@n17.RepeatOut) {also {defaults {Index=1}}}.Element) {also {defaults {AdditionalOperands=2}}}
   HarmonixNodesFFTAnalyzer_0 = HarmonixNodes.FFT Analyzer(In=@AudioMixerAudioMixerStereo2_0.Out L) {also {defaults {Fall Ms=800, Mel Scale Binning="<MetaSoundBuilderResult.SUCCEEDED: 0>", Num Result Bins=64}, version [0,1]}}
   n18 = UE.Reroute(Value=UE.Reroute(Value=UE.Reroute(Value=@HarmonixNodesMusicStemsPlayer_0.On Midi Length Reached).Value).Value)
   TriggerControl_0 = UE.Trigger Control(Close=@ADSREnvelopeADSREnvelope_0.On Attack Triggered, Trigger In=@ADSREnvelopeADSREnvelope_0.On Release Triggered)
   TriggerAnyTriggerAny2_0 = "TriggerAny.Trigger Any (2)"(In 1=UE.Reroute(Value=@ADSREnvelopeADSREnvelope_0.On Done).Value, In 0=@TriggerControl_0.Trigger Out)
   TriggerAnyTriggerAny2_1 = "TriggerAny.Trigger Any (2)"(
    In 1=&MSP_Branch(Value=^bLooping, In=UE.Reroute(Value=@n18.Value).Value).False
    In 0=UE.Reroute(Value=@TriggerAnyTriggerAny2_0.Out).Value
   )
  ]
  links [
   {dst="@Band4", src=Divide_0.Out}
   {dst="@Band3", src=Divide_1.Out}
   {dst="@Band1", src=Divide_2.Out}
   {dst="@Amplitude", src=Divide_3.Out}
   {dst="@Band2", src=Divide_4.Out}
   {dst="@FFTResults", src=HarmonixNodesFFTAnalyzer_0.Results}
   {dst="@MIDI Clock", src=HarmonixNodesMusicStemsPlayer_0.MIDI Clock}
   {dst="@UE.OutputFormat.Quad.Audio:4", src=Reroute_0.Value}
   {dst="@UE.OutputFormat.Quad.Audio:5", src=Reroute_27.Value}
   {dst="@UE.OutputFormat.Quad.Audio:1", src=Reroute_3.Value}
   {dst="@UE.OutputFormat.Quad.Audio:0", src=Reroute_32.Value}
   {dst="@UE.Source.OneShot.OnFinished", src=TriggerAnyTriggerAny2_1.Out}
  ]
 }]
properties {
 assetImportData=&self:AssetImportData
 numChannels=4
 outputFormat=Quad
 sampleRate=48000
 soundClassObject=/Game/Sounds/FortSoundClassesAndModes/Fort_Music_MusicPacks
 virtualizationMode=PlayWhenSilent
 blockRateOverride {default=100}
 modulationSettings {
  volumeRouting=Union
  volumeModulationDestination {bEnableModulation=true, modulators [/Game/Sounds/Modulation/Buses/Sliders/CB_Slider_LobbyMusic]}
 }
 soundSubmixSends [{}]
}
