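# Stereo MetaSound source that plays a Sparks song preview from a media stream, with fades, on the music-packs sound class.
# Playback time drives an ADSR: attack fires at stream start or on a downward crossing of attack time, apparently catching loop restarts.
# Release fires at a time variable minus release time.
# The envelope, multiplied by Volume Adjust Db as linear gain, scales both channels; the Threshold input is unconnected.
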MetaSoundSource /SparksSongPlayer/MetaSounds/MSS_Sparks_Preview
inputs [
 Media Stream Asset {type=MediaStreamAsset}
 Threshold {type=Float}
 UE.Source.OnPlay {type=Trigger}
 Volume Adjust Db {type=Float}
]
interfaces [UE.OutputFormat.Stereo:1.0, UE.Source:1.1]
outputs [
 "UE.OutputFormat.Stereo.Audio:0" {type=Audio}
 "UE.OutputFormat.Stereo.Audio:1" {type=Audio}
]
pages [Default {
  input_defaults {Media Stream Asset=null, Threshold=29, UE.Source.OnPlay=null, Volume Adjust Db=0}
  nodes [
   MediaFrameworkMediaStreamer_0 = MediaFramework.Media Streamer(Media Stream Asset=^Media Stream Asset)
   n1 = UE.ConversionTimeToFloat(In=@MediaFrameworkMediaStreamer_0.Playback Time)
   n2 = VariableAccessor.Time(Variable=InitVariable.Time().Variable)
   TriggerAnyTriggerAny2_0 = "TriggerAny.Trigger Any (2)"(
    In 0=UE.Reroute(Value=@MediaFrameworkMediaStreamer_0.On Stream Starts).Value
    In 1=UE.TriggerOnThreshold:Float(In=@n1.Out, Threshold=UE.ConversionTimeToFloat(In=@n2.Value).Out) {also {defaults {Type=1}}}.Out
   )
   n3 = VariableAccessor.Time(Variable=InitVariable.Time().Variable)
   n4 = UE.Subtract:Time(PrimaryOperand=VariableAccessor.Time(Variable=InitVariable.Time().Variable).Value, AdditionalOperands=@n3.Value) {also {version [1,1]}}
   ADSREnvelopeADSREnvelope_0 = ADSR Envelope.ADSR Envelope:Float(
    Trigger Attack=@TriggerAnyTriggerAny2_0.Out
    Trigger Release=UE.TriggerOnThreshold:Float(In=@n1.Out, Threshold=UE.ConversionTimeToFloat(In=@n4.Out).Out).Out
    Attack Time=VariableAccessor.Time(Variable=@n2.Variable).Value
    Release Time=VariableAccessor.Time(Variable=@n3.Variable).Value
   ) {also {defaults {Decay Time=0, Sustain Level=1}}}
   DecibelstoLinearGain_0 = UE.Decibels to Linear Gain:Float(Decibels=^Volume Adjust Db)
   n5 = VariableAccessor.Float(Variable=VariableMutator.Float(
     Variable=InitVariable.Float().Variable
     Value=UE.Multiply:Float(AdditionalOperands=@ADSREnvelopeADSREnvelope_0.Out Envelope, PrimaryOperand=@DecibelstoLinearGain_0.Linear Gain).Out
    ).Variable)
   Multiply_0 = UE.Multiply:Audio by Float(PrimaryOperand=UE.Reroute(Value=@MediaFrameworkMediaStreamer_0.Out 1).Value, AdditionalOperands=@n5.Value)
   Multiply_1 = UE.Multiply:Audio by Float(PrimaryOperand=UE.Reroute(Value=@MediaFrameworkMediaStreamer_0.Out 0).Value, AdditionalOperands=@n5.Value)
  ]
  links [{dst="@UE.OutputFormat.Stereo.Audio:1", src=Multiply_0.Out}, {dst="@UE.OutputFormat.Stereo.Audio:0", src=Multiply_1.Out}]
 }]
properties {
 assetImportData=&self:AssetImportData
 bLooping=true
 duration=10000
 numChannels=2
 outputFormat=Stereo
 soundClassObject=/Game/Sounds/FortSoundClassesAndModes/Fort_Music_MusicPacks
}
