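# Stereo MetaSound source that plays a Sparks stems song as Fortnite music, with seek, speed and fades.
# A Harmonix stems player drives the transport and MIDI clock output, while audible audio comes from a stereo bus reader rather than the player.
# An ADSR envelope multiplied by decibel volume scales both channels.
# It finishes on envelope completion, fade-out before any attack, or MIDI end when not looping.
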MetaSoundSource /SparksSongPlayer/MetaSounds/MSS_Sparks_Stereo
inputs [
 bLooping {type=Bool}
 Fade In Duration {type=Time}
 Fade Out Duration {type=Time}
 Full Stereo Bus {type=AudioBusAsset}
 Seek Target {type=MusicSeekTarget}
 Speed {type=Float}
 Stems Asset {type=StemsAsset, access=Constructor}
 Trigger Seek {type=Trigger}
 TriggerFadeOut {type=Trigger}
 UE.Source.OnPlay {type=Trigger}
 Volume Adjust Db {type=Float}
]
interfaces [UE.OutputFormat.Stereo:1.0, UE.Source.OneShot:1.0, UE.Source:1.1]
outputs [
 MIDI Clock {type=MIDIClock}
 "UE.OutputFormat.Stereo.Audio:0" {type=Audio}
 "UE.OutputFormat.Stereo.Audio:1" {type=Audio}
 UE.Source.OneShot.OnFinished {type=Trigger}
]
pages [Default {
  input_defaults {
   Fade In Duration=0.15
   Fade Out Duration=0.5
   Full Stereo Bus=/SparksSongPlayer/MetaSounds/Buses/Pre-Effect/FullStereoBus
   Seek Target=null
   Speed=1
   Stems Asset=null
   Trigger Seek=null
   TriggerFadeOut=null
   UE.Source.OnPlay=null
   Volume Adjust Db=0
   bLooping=null
  }
  nodes [
   AudioBusReader2_0 = "UE.Audio Bus Reader (2)"(Audio Bus=^Full Stereo Bus)
   DecibelstoLinearGain_0 = UE.Decibels to Linear Gain:Float(Decibels=^Volume Adjust Db)
   HarmonixNodesMusicStemsPlayer_0 = HarmonixNodes.Music Stems Player(
    Speed=^Speed
    Stems Asset=^Stems Asset
    Transport=HarmonixNodes.TriggerToTransport(Seek Target=^Seek Target, Trigger Seek=^Trigger Seek, Play=^UE.Source.OnPlay).Transport
   )
   ADSREnvelopeADSREnvelope_0 = ADSR Envelope.ADSR Envelope:Float(
    Attack Time=^Fade In Duration
    Release Time=^Fade Out Duration
    Trigger Release=^TriggerFadeOut
    Trigger Attack=HarmonixNodes.TransportToTrigger(Transport=@HarmonixNodesMusicStemsPlayer_0.Transport) {also {version [0,1]}}.Play
   ) {also {defaults {Attack Curve=2, Decay Time=0, Sustain Level=1}}}
   n1 = UE.Reroute(Value=UE.Reroute(Value=UE.Reroute(Value=@ADSREnvelopeADSREnvelope_0.Out Envelope).Value).Value)
   n2 = VariableAccessor.Float(Variable=VariableMutator.Float(
     Variable=InitVariable.Float().Variable
     Value=UE.Multiply:Float(PrimaryOperand=@DecibelstoLinearGain_0.Linear Gain, AdditionalOperands=@n1.Value).Out
    ).Variable)
   Multiply_0 = UE.Multiply:Audio by Float(PrimaryOperand=@AudioBusReader2_0.Out 0, AdditionalOperands=@n2.Value)
   Multiply_1 = UE.Multiply:Audio by Float(PrimaryOperand=@AudioBusReader2_0.Out 1, AdditionalOperands=@n2.Value)
   n3 = UE.Reroute(Value=UE.Reroute(Value=UE.Reroute(Value=@HarmonixNodesMusicStemsPlayer_0.On Midi Length Reached).Value).Value)
   TriggerControl_0 = UE.Trigger Control(Close=@ADSREnvelopeADSREnvelope_0.On Attack Triggered, Trigger In=@ADSREnvelopeADSREnvelope_0.On Release Triggered)
   TriggerAnyTriggerAny2_1 = "TriggerAny.Trigger Any (2)"(In 1=UE.Reroute(Value=@ADSREnvelopeADSREnvelope_0.On Done).Value, In 0=@TriggerControl_0.Trigger Out)
   TriggerAnyTriggerAny2_0 = "TriggerAny.Trigger Any (2)"(
    In 1=&MSP_Branch(Value=^bLooping, In=UE.Reroute(Value=@n3.Value).Value).False
    In 0=UE.Reroute(Value=@TriggerAnyTriggerAny2_1.Out).Value
   )
  ]
  links [
   {dst="@MIDI Clock", src=HarmonixNodesMusicStemsPlayer_0.MIDI Clock}
   {dst="@UE.OutputFormat.Stereo.Audio:0", src=Multiply_0.Out}
   {dst="@UE.OutputFormat.Stereo.Audio:1", src=Multiply_1.Out}
   {dst="@UE.Source.OneShot.OnFinished", src=TriggerAnyTriggerAny2_0.Out}
  ]
 }]
properties {
 assetImportData=&self:AssetImportData
 numChannels=2
 outputFormat=Stereo
 sampleRate=48000
 soundClassObject=/Game/Sounds/FortSoundClassesAndModes/Fort_Music_MusicPacks
 soundSubmixObject=&Fortnite_SFX
 virtualizationMode=PlayWhenSilent
 modulationSettings {
  volumeRouting=Union
  volumeModulationDestination {bEnableModulation=true, modulators [/Game/Sounds/Modulation/Buses/Sliders/CB_Slider_LobbyMusic]}
 }
 soundSubmixSends [{}]
}
