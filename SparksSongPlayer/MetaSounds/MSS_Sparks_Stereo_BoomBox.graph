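# MetaSound Source preset of MSS_Sparks_Stereo that plays a Sparks song's stems asset in stereo, named for boombox playback.
# The inherited graph exposes looping, speed, seek, fade and volume inputs, and outputs stereo audio, a MIDI clock and a finished trigger.
# It references the pre-effect FullStereoBus, uses the Music Packs sound class with the Fortnite SFX submix, and keeps playing when silent.
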MetaSoundSource /SparksSongPlayer/MetaSounds/MSS_Sparks_Stereo_BoomBox
preset_parent=&/MSS_Sparks_Stereo
inputs [
 bLooping {type=Bool}
 Fade In Duration {type=Time}
 Fade Out Duration {type=Time}
 Full Stereo Bus {type=AudioBusAsset}
 Seek Target {type=MusicSeekTarget}
 Speed {type=Float}
 Stems Asset {type=StemsAsset, access=Constructor}
 Trigger Seek {type=Trigger}
 TriggerFadeOut {type=Trigger}
 UE.Source.OnPlay {type=Trigger}
 Volume Adjust Db {type=Float}
]
interfaces [UE.OutputFormat.Stereo:1.0, UE.Source.OneShot:1.0, UE.Source:1.1]
outputs [
 MIDI Clock {type=MIDIClock}
 "UE.OutputFormat.Stereo.Audio:0" {type=Audio}
 "UE.OutputFormat.Stereo.Audio:1" {type=Audio}
 UE.Source.OneShot.OnFinished {type=Trigger}
]
pages [Default {input_defaults {
   Fade In Duration=0.15
   Fade Out Duration=0.5
   Full Stereo Bus=/SparksSongPlayer/MetaSounds/Buses/Pre-Effect/FullStereoBus
   Seek Target=null
   Speed=1
   Stems Asset=null
   Trigger Seek=null
   TriggerFadeOut=null
   UE.Source.OnPlay=null
   Volume Adjust Db=0
   bLooping=null
  }}]
properties {
 assetImportData=&self:AssetImportData
 numChannels=2
 outputFormat=Stereo
 sampleRate=48000
 soundClassObject=/Game/Sounds/FortSoundClassesAndModes/Fort_Music_MusicPacks
 soundSubmixObject=&Fortnite_SFX
 virtualizationMode=PlayWhenSilent
 modulationSettings {volumeRouting=Union, volumeModulationDestination {bEnableModulation=true}}
 soundSubmixSends [{}]
}
